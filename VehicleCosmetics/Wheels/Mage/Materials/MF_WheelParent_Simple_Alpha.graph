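# Simplified wheel material function for vehicle cosmetics that builds rim and optional tire attributes, with Opacity and OpacityMask taken from diffuse texture alpha.
# Rim and tire parameters each feed MF_Wheel_Parent_Simple_InputSetup, using RGB masks for three color/PBR zones and alpha for AO.
# MF_Wheel_TreadMirroring supplies mirrored tire UVs and normal correction.
# The Tire_Exists static switch enables vertex-color-red blending of rim and tire; masks and normals are also output.
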MaterialFunction /VehicleCosmetics/Wheels/Mage/Materials/MF_WheelParent_Simple_Alpha
graph {
 properties {description=Simplified MF For Wheels}
 nodes [
  Composite() {subgraphName=Color Settings A-C}
  Composite() {subgraphName=Color Settings A-C}
  Composite() {subgraphName=Diffuse Settings w/Tread Mirroring}
  Composite() {subgraphName=Diffuse Settings}
  Composite() {subgraphName=Normal w/Tread Mirroring}
  Composite() {subgraphName=Normal - No Tread Mirroring}
  Composite() {subgraphName=Default Settings}
  Composite() {subgraphName=Default Settings}
  Composite() {subgraphName="Color && Mask"}
  Composite() {subgraphName="Color && Mask w/Tread Mirroring"}
  FunctionOutput(None=NamedRerouteUsage()) {description=RGBA channel outputs for Rim masking., outputName=RimRGBAMask, sortPriority=1}
  FunctionOutput(None=StaticSwitchParameter(True=NamedRerouteUsage(), False=Constant() {desc="Default to prevent errors if\r\ntire doesn't exist."}) {desc="Tire == True, No Tire == False. False is default.", group=5_Special Effects, parameterName=Tire_Exists}) {description=RGBA channel outputs for Tire masking., outputName=TireRGBAMask, sortPriority=2}
  n1 = ScalarParameter() {
   defaultValue=0.4
   desc=General Roughness that takes effect in absence of masking.
   group=4_Surface Defaults
   parameterName=Rim_RoughnessBase
   sliderMax=1
  }
  n2 = ScalarParameter() {
   defaultValue=1
   desc=General Metalness that takes effect in absence of masking.
   group=4_Surface Defaults
   parameterName=Rim_MetallicBase
   sliderMax=1
  }
  n3 = ScalarParameter() {
   defaultValue=1
   desc="0==False, 1==True: AO gradient affects Emissive values. Use this for Emissives with depth. Values between 0 && 1 are allowed."
   group=5_Special Effects
   parameterName=Rim_AOTintsEmissive
  }
  n4 = ScalarParameter() {
   desc="0==False, 1==True: Blends AO with Diffuse. Values between 0 && 1 are allowed."
   group=5_Special Effects
   parameterName=Rim_AOTintsDiffuse
  }
  n5 = MaterialFunctionCall(R=@n1, G=@n2, B=@n3, A=@n4) {desc="Vector4 container for Rough, Metallic, AO settings", materialFunction=&AppendMany}
  MaterialExpressionReroute_16 = Reroute(None=@n5.RGBA)
  n6 = TextureSampleParameter2D() {group=1_Input Textures, parameterName=Rim_Normal, samplerType=SAMPLERTYPE_Normal, texture=/VehicleCosmetics/Textures/T_BlankNormal_B}
  MaterialExpressionReroute_50 = Reroute(None=@n6.RGB)
  n7 = TextureSampleParameter2D() {desc=Diffuse Input Texture - Optional, group=1_Input Textures, parameterName=Rim_DiffuseTexture, texture=/VehicleCosmetics/Textures/T_White}
  n8 = ScalarParameter() {
   desc="0==False, 1==True: Set at 1 if Diffuse overrides color, 0 for no Diffuse. Auto-rounds to 0 or 1."
   group=1_Input Textures
   parameterName=Rim_DiffuseTextureEnable
   sliderMax=1
  }
  n9 = VectorParameter() {
   desc="Albedo/Diffuse color control in absence of RGB maps, or to fill in the gaps. Leave white if using Diffuse Texture instead."
   group=4_Surface Defaults
   parameterName=Rim_ColorBase
   defaultValue=<rgba 1 1 1 1>
  }
  MaterialExpressionReroute_17 = Reroute(None=@n9.RGB)
  n10 = TextureSampleParameter2D() {
   desc=Mask for Rim - Alpha is AO
   group=1_Input Textures
   parameterName=Rim_MaskRGBA
   samplerType=SAMPLERTYPE_Masks
   texture=/VehicleCosmetics/Textures/T_Black_Masks
   channelNames {a=AO, b=ColorC, g=ColorB, r=ColorA}
  }
  MaterialExpressionReroute_18 = Reroute(None=@n10.RGBA)
  n11 = VectorParameter() {
   desc=Color fill for R value of RGB Mask
   group=2_Rim Properties
   parameterName=Rim_ColorA
   channelNames {a=Unused, b="", g="", r=""}
   defaultValue=<rgba 1 0 0 1>
  }
  MaterialExpressionReroute_1 = Reroute(None=@n11)
  n12 = VectorParameter() {
   desc="Emissive, Roughness, Metallic values for ColorA. Emissive has no limit, Roughness && Metallic are clamped between 0 && 1."
   group=2_Rim Properties
   parameterName=Rim_ColorA_PBR
   channelNames {a=Unused, b=Metallic, g=Roughness, r=Emissive}
   defaultValue=<rgba 0 0.4 1 1>
  }
  MaterialExpressionReroute_2 = Reroute(None=@n12)
  n13 = VectorParameter() {
   desc=Color fill for G value of RGB Mask
   group=2_Rim Properties
   parameterName=Rim_ColorB
   channelNames {a=Unused, b="", g="", r=""}
   defaultValue=<rgba 0 1 0 1>
  }
  MaterialExpressionReroute_3 = Reroute(None=@n13)
  n14 = VectorParameter() {
   desc="Emissive, Roughness, Metallic values for ColorB. Emissive has no limit, Roughness && Metallic are clamped between 0 && 1."
   group=2_Rim Properties
   parameterName=Rim_ColorB_PBR
   channelNames {a=Unused, b=Metallic, g=Roughness, r=Emissive}
   defaultValue=<rgba 0 0.4 1 1>
  }
  MaterialExpressionReroute_6 = Reroute(None=@n14)
  n15 = VectorParameter() {
   desc=Color fill for B value of RGB Mask
   group=2_Rim Properties
   parameterName=Rim_ColorC
   channelNames {a=Unused, b="", g="", r=""}
   defaultValue=<rgba 0 0 1 1>
  }
  MaterialExpressionReroute_7 = Reroute(None=@n15)
  n16 = VectorParameter() {
   desc="Emissive, Roughness, Metallic values for ColorC. Emissive has no limit, Roughness && Metallic are clamped between 0 && 1."
   group=2_Rim Properties
   parameterName=Rim_ColorC_PBR
   channelNames {a=Unused, b=Metallic, g=Roughness, r=Emissive}
   defaultValue=<rgba 0 0.4 1 1>
  }
  MaterialExpressionReroute_8 = Reroute(None=@n16)
  n17 = MaterialFunctionCall!(
   Base_Settings=@MaterialExpressionReroute_16.RoughMetalAOTint
   Normal_Texture=@MaterialExpressionReroute_50.RGB
   Diffuse_Texture=@n7.RGB
   DiffuseTextureEnable=@n8
   ColorBase=@MaterialExpressionReroute_17.Rim_ColorBase
   MaskRGBA=@MaterialExpressionReroute_18.Rim_MaskRGBA
   ColorA_Color=@MaterialExpressionReroute_1.ColorA
   ColorA_PBR_Values=@MaterialExpressionReroute_2.ColorA_PBR
   ColorB_Color=@MaterialExpressionReroute_3.ColorB
   ColorB_PBR_Values=@MaterialExpressionReroute_6.ColorB_PBR
   ColorC_Color=@MaterialExpressionReroute_7.ColorC
   ColorC_PBR_Values=@MaterialExpressionReroute_8.ColorC_PBR
  ) {materialFunction=/VehicleCosmetics/Wheels/_SharedMaterials/MF_Wheel_Parent_Simple_InputSetup}
  n18 = ScalarParameter() {
   defaultValue=0.7
   desc=General Roughness that takes effect in absence of masking.
   group=4_Surface Defaults
   parameterName=Tire_RoughnessBase
   sliderMax=1
  }
  n19 = ScalarParameter() {desc=General Metalness that takes effect in absence of masking., group=4_Surface Defaults, parameterName=Tire_MetallicBase, sliderMax=1}
  n20 = ScalarParameter() {
   defaultValue=1
   desc="0==False, 1==True: AO gradient affects Emissive values. Use this for Emissives with depth. Values between 0 && 1 are allowed."
   group=5_Special Effects
   parameterName=Tire_AOTintsEmissive
  }
  n21 = ScalarParameter() {
   desc="0==False, 1==True: Blends AO with Diffuse. Values between 0 && 1 are allowed."
   group=5_Special Effects
   parameterName=Tire_AOTintsDiffuse
  }
  n22 = MaterialFunctionCall(R=@n18, G=@n19, B=@n20, A=@n21) {desc="Vector4 container for Rough, Metallic, AO settings", materialFunction=&AppendMany}
  MaterialExpressionReroute_15 = Reroute(None=@n22.RGBA)
  n23 = TextureSampleParameter2D(UVs=NamedRerouteUsage()) {group=1_Input Textures, parameterName=Tire_Normal, samplerType=SAMPLERTYPE_Normal, texture=&T_BlankNormal_B}
  n24 = MaterialFunctionCall(R=Multiply(A=@n23.R, B=NamedRerouteUsage()) {constB=-1}, G=Multiply(A=@n23.G, B=NamedRerouteUsage()) {constB=-1}, B=@n23.B) {materialFunction=&AppendMany}
  MaterialExpressionReroute_49 = Reroute(None=@n24.RGB)
  n25 = TextureSampleParameter2D(UVs=NamedRerouteUsage()) {desc=Diffuse Input Texture - Optional, group=1_Input Textures, parameterName=Tire_DiffuseTexture, texture=&T_White}
  MaterialExpressionReroute_44 = Reroute(None=@n25.RGB)
  n26 = ScalarParameter() {
   desc="0==False, 1==True: Set at 1 if Diffuse overrides color, 0 for no Diffuse. Auto-rounds to 0 or 1."
   group=1_Input Textures
   parameterName=Tire_DiffuseTextureEnable
   sliderMax=1
  }
  MaterialExpressionReroute_45 = Reroute(None=@n26)
  n27 = VectorParameter() {
   desc="Albedo/Diffuse color control in absence of RGB maps, or to fill in the gaps. Leave white if using Diffuse Texture instead."
   group=4_Surface Defaults
   parameterName=Tire_ColorBase
   defaultValue=<rgba 0.02 0.02 0.02 1>
  }
  MaterialExpressionReroute_19 = Reroute(None=@n27.RGB)
  n28 = TextureSampleParameter2D(UVs=NamedRerouteUsage()) {
   desc=Mask for Tire - Alpha is AO
   group=1_Input Textures
   parameterName=Tire_MaskRGBA
   samplerType=SAMPLERTYPE_Masks
   texture=&T_Black_Masks
   channelNames {a=AO, b=ColorC, g=ColorB, r=ColorA}
  }
  MaterialExpressionReroute_20 = Reroute(None=@n28.RGBA)
  n29 = VectorParameter() {
   desc=Color fill for R value of RGB Mask
   group=3_Tire Properties
   parameterName=Tire_ColorA
   channelNames {a=Unused, b="", g="", r=""}
   defaultValue=<rgba 1 0.1248 0 1>
  }
  MaterialExpressionReroute_9 = Reroute(None=@n29)
  n30 = VectorParameter() {
   desc="Emissive, Roughness, Metallic values for ColorA. Emissive has no limit, Roughness && Metallic are clamped between 0 && 1."
   group=3_Tire Properties
   parameterName=Tire_ColorA_PBR
   channelNames {a=Unused, b=Metallic, g=Roughness, r=Emissive}
   defaultValue=<rgba 0 0.7 0 1>
  }
  MaterialExpressionReroute_10 = Reroute(None=@n30)
  n31 = VectorParameter() {
   desc=Color fill for G value of RGB Mask
   group=3_Tire Properties
   parameterName=Tire_ColorB
   channelNames {a=Unused, b="", g="", r=""}
   defaultValue=<rgba 0 1 0.5029 1>
  }
  MaterialExpressionReroute_11 = Reroute(None=@n31)
  n32 = VectorParameter() {
   desc="Emissive, Roughness, Metallic values for ColorB. Emissive has no limit, Roughness && Metallic are clamped between 0 && 1."
   group=3_Tire Properties
   parameterName=Tire_ColorB_PBR
   channelNames {a=Unused, b=Metallic, g=Roughness, r=Emissive}
   defaultValue=<rgba 0 0.7 0 1>
  }
  MaterialExpressionReroute_12 = Reroute(None=@n32)
  n33 = VectorParameter() {
   desc=Color fill for B value of RGB Mask
   group=3_Tire Properties
   parameterName=Tire_ColorC
   channelNames {a=Unused, b="", g="", r=""}
   defaultValue=<rgba 0.5029 0 1 1>
  }
  MaterialExpressionReroute_13 = Reroute(None=@n33)
  n34 = VectorParameter() {
   desc="Emissive, Roughness, Metallic values for ColorC. Emissive has no limit, Roughness && Metallic are clamped between 0 && 1."
   group=3_Tire Properties
   parameterName=Tire_ColorC_PBR
   channelNames {a=Unused, b=Metallic, g=Roughness, r=Emissive}
   defaultValue=<rgba 0 0.7 0 1>
  }
  MaterialExpressionReroute_14 = Reroute(None=@n34)
  n35 = MaterialFunctionCall(
   Base_Settings=@MaterialExpressionReroute_15.RoughMetalAOTint
   Normal_Texture=@MaterialExpressionReroute_49.Tire_Normal
   Diffuse_Texture=@MaterialExpressionReroute_44.Tire_DiffuseTexture
   DiffuseTextureEnable=@MaterialExpressionReroute_45.Tire_DiffuseTextureEnable
   ColorBase=@MaterialExpressionReroute_19.Tire_ColorBase
   MaskRGBA=@MaterialExpressionReroute_20.Tire_MaskRGBA
   ColorA_Color=@MaterialExpressionReroute_9.ColorA
   ColorA_PBR_Values=@MaterialExpressionReroute_10.ColorA_PBR
   ColorB_Color=@MaterialExpressionReroute_11.ColorB
   ColorB_PBR_Values=@MaterialExpressionReroute_12.ColorB_PBR
   ColorC_Color=@MaterialExpressionReroute_13.ColorC
   ColorC_PBR_Values=@MaterialExpressionReroute_14.ColorC_PBR
  ) {materialFunction=&MF_Wheel_Parent_Simple_InputSetup}
  n36 = BlendMaterialAttributes(A=@n17.Result, B=@n35.Result, Alpha=NamedRerouteDeclaration(None=VertexColor().R) {name=VertColor_R_Out, nodeColor=<rgba 1 0.5 0 1>})
  n37 = TextureSampleParameter2D() {desc=Diffuse Input Texture - Optional, group=1_Input Textures, parameterName=Tire_DiffuseTexture, texture=&T_White}
  n38 = LinearInterpolate(A=@n7.A, B=@n37.A, Alpha=VertexColor().R)
  FunctionOutput(None=StaticSwitchParameter(
    True=SetMaterialAttributes!(MaterialAttributes=@n36, Opacity=@n38, OpacityMask=@n38) {attributeSetTypes [#1, #2]}
    False=SetMaterialAttributes(MaterialAttributes=@n17.Result, Opacity=@n7.A, OpacityMask=@n7.A) {attributeSetTypes [#1, #2]}
   ) {desc="Tire == True, No Tire == False. False is default.", group=5_Special Effects, parameterName=Tire_Exists}) {description=Base material attributes for WheelParent_Simple., outputName=MaterialAttributes}
  FunctionOutput(None=StaticSwitchParameter(True=NamedRerouteUsage(), False=Constant() {desc="Default to prevent errors if\r\ntire doesn't exist."}) {desc="Tire == True, No Tire == False. False is default.", group=5_Special Effects, parameterName=Tire_Exists}) {description=RGBA channel outputs for Tire masking., outputName=VertexColorR, sortPriority=3}
  FunctionOutput(None=@MaterialExpressionReroute_49.Tire_Normal) {description=RGBA channel outputs for Rim masking., outputName=NormalTire, sortPriority=1}
  FunctionOutput(None=@MaterialExpressionReroute_50.RGB) {description=RGBA channel outputs for Rim masking., outputName=NormalRim, sortPriority=1}
  n39 = MaterialFunctionCall() {
   desc="UV - For all Tread Mirrors. Lerp A && B - For Normals"
   materialFunction=/VehicleCosmetics/Wheels/_SharedMaterials/MF_Wheel_TreadMirroring
  }
  NamedRerouteDeclaration(None=@n39.LerpA) {name=Tread_LerpA, nodeColor=<rgba 0 1 0.6953 1>}
  NamedRerouteDeclaration(None=@n35.Masks Out) {name=MaskTire_Out, nodeColor=<rgba 0.4072 0.1046 0 1>}
  NamedRerouteDeclaration(None=@n17.Masks Out) {name=MaskRim_Out, nodeColor=<rgba 0.0409 0.0666 0.2918 1>}
  NamedRerouteDeclaration(None=@n39.LerpB) {name=Tread_LerpB, nodeColor=<rgba 0 1 0.6953 1>}
  NamedRerouteDeclaration(None=@n39.UV) {name=Tread_UV, nodeColor=<rgba 0 1 0.6953 1>}
  PinBase()
  PinBase!(ColorA=@n11, ColorA_PBR=@n12, ColorB=@n13, ColorB_PBR=@n14, ColorC=@n15, ColorC_PBR=@n16) {reroutePins [
    ColorA {expression=&self:MaterialExpressionReroute_1}
    ColorA_PBR {expression=&self:MaterialExpressionReroute_2}
    ColorB {expression=&self:MaterialExpressionReroute_3}
    ColorB_PBR {expression=&self:MaterialExpressionReroute_6}
    ColorC {expression=&self:MaterialExpressionReroute_7}
    ColorC_PBR {expression=&self:MaterialExpressionReroute_8}
   ]}
  PinBase!()
  PinBase!(Tire_ColorBase=@n27.RGB, Tire_MaskRGBA=@n28.RGBA) {reroutePins [
    Tire_ColorBase {expression=&self:MaterialExpressionReroute_19}
    Tire_MaskRGBA {expression=&self:MaterialExpressionReroute_20}
   ]}
  PinBase!()
  PinBase()
  PinBase!(Tire_DiffuseTexture=@n25.RGB, Tire_DiffuseTextureEnable=@n26) {reroutePins [
    Tire_DiffuseTexture {expression=&self:MaterialExpressionReroute_44}
    Tire_DiffuseTextureEnable {expression=&self:MaterialExpressionReroute_45}
   ]}
  PinBase!()
  n40 = TextureSampleParameter2D() {desc=Diffuse Input Texture - Optional, group=1_Input Textures, parameterName=Rim_DiffuseTexture, texture=&T_White}
  n41 = ScalarParameter() {
   desc="0==False, 1==True: Set at 1 if Diffuse overrides color, 0 for no Diffuse. Auto-rounds to 0 or 1."
   group=1_Input Textures
   parameterName=Rim_DiffuseTextureEnable
   sliderMax=1
  }
  PinBase!(Rim_DiffuseTexture=@n40.RGB, Rim_DiffuseTextureEnable=@n41) {reroutePins [
    Rim_DiffuseTexture {expression=&self:MaterialExpressionReroute_46}
    Rim_DiffuseTextureEnable {expression=&self:MaterialExpressionReroute_47}
   ]}
  PinBase!()
  PinBase!(Tire_Normal=@n24.RGB) {reroutePins [Tire_Normal {expression=&self:MaterialExpressionReroute_49}]}
  PinBase!(ColorA=@n29, ColorA_PBR=@n30, ColorB=@n31, ColorB_PBR=@n32, ColorC=@n33, ColorC_PBR=@n34) {reroutePins [
    ColorA {expression=&self:MaterialExpressionReroute_9}
    ColorA_PBR {expression=&self:MaterialExpressionReroute_10}
    ColorB {expression=&self:MaterialExpressionReroute_11}
    ColorB_PBR {expression=&self:MaterialExpressionReroute_12}
    ColorC {expression=&self:MaterialExpressionReroute_13}
    ColorC_PBR {expression=&self:MaterialExpressionReroute_14}
   ]}
  PinBase!()
  PinBase!(RGB=@n6.RGB) {reroutePins [RGB {expression=&self:MaterialExpressionReroute_50}]}
  PinBase!()
  PinBase!(RoughMetalAOTint=@n22.RGBA) {reroutePins [RoughMetalAOTint {expression=&self:MaterialExpressionReroute_15}]}
  PinBase!()
  PinBase!(RoughMetalAOTint=@n5.RGBA) {reroutePins [RoughMetalAOTint {expression=&self:MaterialExpressionReroute_16}]}
  PinBase!()
  PinBase!(Rim_ColorBase=@n9.RGB, Rim_MaskRGBA=@n10.RGBA) {reroutePins [
    Rim_ColorBase {expression=&self:MaterialExpressionReroute_17}
    Rim_MaskRGBA {expression=&self:MaterialExpressionReroute_18}
   ]}
  MaterialExpressionReroute_46 = Reroute(None=@n40.RGB)
  MaterialExpressionReroute_47 = Reroute(None=@n41)
 ]
}
