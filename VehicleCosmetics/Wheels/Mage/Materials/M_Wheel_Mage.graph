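# M_Wheel_Mage is the masked, default-lit material for the Mage vehicle wheel cosmetic, enabled for skeletal meshes and mesh deformers.
# It uses material attributes to chain three material functions, with no other logic in the graph.
# MF_WheelParent_Simple_Alpha supplies the base wheel attributes, which pass through MF_BaseVehicle_Damage and then MF_VehicleDeRez_Transition to layer on vehicle damage and de-rez effects.
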Material /VehicleCosmetics/Wheels/Mage/Materials/M_Wheel_Mage
graph {
 properties {
  bAutomaticallySetUsageInEditor=false
  bUseMaterialAttributes=true
  bUsedWithMeshDeformer=true
  bUsedWithSkeletalMesh=true
  blendMode=BLEND_Masked
  pixelDepthOffsetMode=PDOM_Legacy
  usedShadingModels=Default Lit
 }
 nodes [MaterialFunctionCall(In=MaterialFunctionCall(Material=MaterialFunctionCall() {materialFunction=&/MF_WheelParent_Simple_Alpha}.MaterialAttributes) {materialFunction=&MF_BaseVehicle_Damage}.Result) {materialFunction=&MF_VehicleDeRez_Transition}]
}
