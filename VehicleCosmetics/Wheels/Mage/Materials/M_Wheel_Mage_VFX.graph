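# Unlit emissive material for the Mage wheel VFX, showing a green-to-cyan dragon within animated fire, usable on skeletal meshes.
# Rotating noise layers build a Warp signal, while a flow map panned in radial coordinates distorts a radial gradient into fire masks.
# The dragon texture gets UV distortion and bump-offset depth, with a static switch governing vertex-red blending.
# Named reroutes carry these signals into nested color lerps scaled by EmissivePower.
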Material /VehicleCosmetics/Wheels/Mage/Materials/M_Wheel_Mage_VFX
graph {
 properties {
  bAutomaticallySetUsageInEditor=false
  bUsedWithMeshDeformer=true
  bUsedWithSkeletalMesh=true
  shadingModel=MSM_Unlit
  usedShadingModels=Unlit
 }
 nodes [
  n1 = TextureSample(UVs=Rotator() {speed=-0.6}) {texture=/VehicleCosmetics/Wheels/Mage/Textures/T_Hotto_NoiseRGB1}
  n2 = Multiply(A=@n1.B, B=Constant() {r=0.4})
  n3 = TextureCoordinate() {uTiling=2, vTiling=2}
  n4 = Multiply(A=TextureSample(UVs=@n3) {texture=&T_Hotto_NoiseRGB1}.G, B=Constant() {r=0.3})
  n5 = TextureSample(UVs=Add(A=Rotator(Coordinate=@n3) {speed=0.1}, B=Add(A=@n2, B=@n4))) {texture=/VehicleCosmetics/Wheels/StickyPurple/Textures/wheel_StickyPurple_Noise1}
  NamedRerouteDeclaration(None=Multiply(A=Add(A=@n1.B, B=@n1.B), B=Add(A=@n5.R, B=@n5.R))) {name=Warp, nodeColor=<rgba 0 0.25 1 1>}
  n6 = ComponentMask(None=Add(A=TextureCoordinate(), B=Constant() {r=-0.5})) {g=true, r=true}
  n7 = ConstantBiasScale(None=DotProduct(A=@n6, B=@n6)) {bias=0, scale=20}
  n8 = TextureCoordinate()
  n9 = TextureSample() {samplerType=SAMPLERTYPE_LinearColor, texture=/VehicleCosmetics/Wheels/_SharedTextures/FX/Radial_Coords01_D}
  n10 = Panner(Coordinate=Multiply(A=ComponentMask(None=@n9.RGB) {g=true, r=true}, B=Constant2Vector() {g=1, r=2})) {speedX=-0.2, speedY=-0.05}
  n11 = Multiply(
   A=ComponentMask(None=TextureSampleParameter2D(UVs=@n10) {parameterName=Flow03, samplerType=SAMPLERTYPE_LinearColor, texture=/VehicleCosmetics/Wheels/Mage/Textures/T_Flow_03_nomips}.RGB) {g=true, r=true}
   B=Constant() {r=1}
  )
  n12 = ComponentMask(None=Add(A=Add(A=@n8, B=@n11), B=Constant() {r=-1})) {g=true, r=true}
  n13 = DotProduct(A=@n12, B=@n12)
  NamedRerouteDeclaration(None=Multiply(A=Clamp(None=@n7), B=@n13)) {name=Fire, nodeColor=<rgba 1 0.355 0.0349 1>}
  n14 = Multiply(A=@n13, B=ScalarParameter() {defaultValue=2, group=Dragon, parameterName=OuterFire_Sharpness})
  NamedRerouteDeclaration(None=Clamp(None=ConstantBiasScale(None=@n14) {bias=-0.1, scale=5})) {name=FireClamp, nodeColor=<rgba 0.1719 1 0 1>}
  n15 = LinearInterpolate(A=Constant3Vector() {constant=<rgba 0 0.46 0.176 0>}, B=Constant3Vector() {constant=<rgba 0.003 0.9889 0.83 0>}, Alpha=NamedRerouteUsage())
  n16 = TextureCoordinate()
  n17 = Multiply(A=NamedRerouteUsage(), B=ScalarParameter() {defaultValue=0.1, group=Dragon, parameterName=Dragon_WarpAmount})
  n18 = TextureSampleParameter2D(UVs=BumpOffset(Coordinate=Add(A=@n16, B=@n17), Height=ScalarParameter() {defaultValue=-4, group=Dragon, parameterName=DragonDepth})) {parameterName=Dragon_Texture, texture=/VehicleCosmetics/Wheels/Mage/Textures/T_Mage_Dragon_01}
  n19 = Constant() {r=1}
  n20 = VertexColor()
  n21 = Multiply(
   A=StaticSwitchParameter(True=@n18.R, False=LinearInterpolate(A=@n18.R, B=@n19, Alpha=@n20.R)) {defaultValue=true, parameterName=Dragon_VertexMask}
   B=OneMinus(None=NamedRerouteUsage())
  )
  n22 = LinearInterpolate(
   A=Constant3Vector() {constant=<rgba 0.0002 0.138 0.0014 0.0471>}
   B=Constant3Vector() {constant=<rgba 0 0.91 0.27 0>}
   Alpha=NamedRerouteUsage()
  )
  NamedRerouteDeclaration(None=Multiply(A=LinearInterpolate(A=@n22, B=@n15, Alpha=@n21), B=ScalarParameter() {defaultValue=2.5, group=Dragon, parameterName=EmissivePower})) {name=Emissive, nodeColor=<rgba 0.814 0.2884 1 1>}
  NamedRerouteUsage()
 ]
}
